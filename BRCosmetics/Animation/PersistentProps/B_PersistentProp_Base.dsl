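; Base persistent prop Blueprint in the cosmetics content that handles a prop's ambient and removal audio and Niagara effects.
; Ambient sound and VFX attach to the root at BeginPlay or on detach, chosen separately by per-effect spawn flags.
; DestroyProp either destroys instantly or picks despawn versus derez effects by whether remaining lifetime has run out.
; The non-instant path fades out ambient audio and destroys the actor after a short delay.
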(event EventBeginPlay
  (Utilities|FlowControl|Sequence
    (:then_0
      (if (and (Variables|Audio|GetbShouldPlayAmbientAudioOnSpawn) (Utilities|IsValid (Variables|Audio|GetAmbient%20Audio) :!class /Script/Engine.KismetSystemLibrary))
        (bind _returnvalue (Audio|SpawnSoundAttached (Variables|Audio|GetAmbient%20Audio) (Variables|Default|GetDefaultSceneRoot) "root" (Variables|Audio|GetSFXLocalOffset)))
        (bind _output_get (Variables|Default|SetAmbientAudioComponent _returnvalue))
        (Audio|Components|Audio|FadeIn _output_get 0.1 1.0 0.0 "Linear" :!class /Script/Engine.AudioComponent)))
    (:then_1
      (if (and (Variables|VFX|GetbShouldPlayAmbientVFXOnSpawn) (Utilities|IsValid (Variables|VFX|GetAmbient%20VFX) :!class /Script/Engine.KismetSystemLibrary))
        (Niagara|SpawnSystemAttached (Variables|VFX|GetAmbient%20VFX) (Variables|Default|GetDefaultSceneRoot) "root" (Variables|VFX|GetFXLocalOffset) "0, 0, 0" "SnapToTarget")))))

(event EventOnDetach
  (Utilities|FlowControl|Sequence
    (:then_0
      (if (and (not (Variables|Audio|GetbShouldPlayAmbientAudioOnSpawn)) (Utilities|IsValid (Variables|Audio|GetAmbient%20Audio) :!class /Script/Engine.KismetSystemLibrary))
        (bind _returnvalue (Audio|SpawnSoundAttached (Variables|Audio|GetAmbient%20Audio) (Variables|Default|GetDefaultSceneRoot) "root" (Variables|Audio|GetSFXLocalOffset)))
        (bind _output_get (Variables|Default|SetAmbientAudioComponent _returnvalue))
        (Audio|Components|Audio|FadeIn _output_get 0.1 1.0 0.0 "Linear" :!class /Script/Engine.AudioComponent)))
    (:then_1
      (if (and (not (Variables|VFX|GetbShouldPlayAmbientVFXOnSpawn)) (Utilities|IsValid (Variables|VFX|GetAmbient%20VFX) :!class /Script/Engine.KismetSystemLibrary))
        (Niagara|SpawnSystemAttached (Variables|VFX|GetAmbient%20VFX) (Variables|Default|GetDefaultSceneRoot) "root" (Variables|VFX|GetFXLocalOffset))))))

(event EventDestroyProp
  (params (bShouldInstantDestroy :type Boolean))
  (bind _remaininglifetime (Variables|Lifetime|GetRemainingLifetime))
  (bind _returnvalue (<= _remaininglifetime 0))
  (bind _remaininglifetime_2 (Variables|Lifetime|GetRemainingLifetime))
  (bind _returnvalue_2 (<= _remaininglifetime_2 0))
  (bind _despawn_vfx (Variables|VFX|GetDespawn%20VFX))
  (bind _derez_vfx (Variables|VFX|GetDerez%20VFX))
  (bind _returnvalue_3 (select _returnvalue _despawn_vfx _derez_vfx))
  (bind _despawn_audio (Variables|Audio|GetDespawn%20Audio))
  (bind _derezaudio (Variables|Audio|GetDerezAudio))
  (bind _returnvalue_4 (select _returnvalue_2 _despawn_audio _derezaudio))
  (bind _ambientaudiocomponent (Variables|Default|GetAmbientAudioComponent))
  (if bShouldInstantDestroy
    (Actor|DestroyActor)
    (else
      (Utilities|FlowControl|Sequence
        (:then_0
          (Utilities|IsValid _returnvalue_3
            (:Is%20Valid
              (Niagara|SpawnSystematLocation _returnvalue_3 (Math|Transform|TransformLocation (Transformation|GetWorldTransform (Variables|Default|GetDefaultSceneRoot)) (Variables|VFX|GetFXLocalOffset)) (Transformation|GetActorRotation)))))
        (:then_1
          (Utilities|IsValid _returnvalue_4
            (:Is%20Valid
              (bind _returnvalue_5 (Audio|SpawnSoundatLocation _returnvalue_4 (Math|Transform|TransformLocation (Transformation|GetWorldTransform (Variables|Default|GetDefaultSceneRoot)) (Variables|Audio|GetSFXLocalOffset))))
              (Audio|Components|Audio|FadeIn _returnvalue_5 0.1 1.0 0.0 "Linear" :!class /Script/Engine.AudioComponent))))
        (:then_2
          (if (Utilities|IsValid _ambientaudiocomponent :!class /Script/Engine.KismetSystemLibrary)
            (Audio|Components|Audio|FadeOut _ambientaudiocomponent 0.1 0.0 "Linear" :!class /Script/Engine.AudioComponent)
            (Utilities|FlowControl|Delay 0.2)
            (Components|DestroyComponent (Variables|Default|GetAmbientAudioComponent))))
        (:then_3
          (Utilities|FlowControl|Delay 0.5)
          (Actor|DestroyActor))))))

(fn ConstructionScript)
