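; Skeletal persistent emote prop that animates its own skeletal mesh in sync with the owning pawn's emote montage.
; On BeginPlay it picks a montage by the player's body type and gender, falling back to medium male, then applies the mesh and optional anim blueprint.
; If that montage's length equals the owner's active montage, it plays them synced and flags the emote as started.
; Missing medium male montages or length mismatches print on-screen errors.
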(event EventBeginPlay
  (|CallParentFunction)
  (bind _animation_montage (CallFunction|DetermineAnimationMontage))
  (Variables|Default|SetAnimation%20Montage _animation_montage)
  (Components|SkinnedMesh|SetSkinnedAssetandUpdate (Variables|Default|GetSkeletalMeshComponent) (Variables|Default|GetSkeletal%20Mesh) true)
  (if (Utilities|IsValidClass (Variables|Default|GetAnimation%20Blueprint))
    (Components|SkeletalMesh|SetAnimInstanceClass (Variables|Default|GetSkeletalMeshComponent) (Variables|Default|GetAnimation%20Blueprint))
    (else
      _))
  (Animation|Utilities|ForceTickAnimation (Variables|Default|GetSkeletalMeshComponent))
  (Utilities|IsValid (Variables|Default|GetAnimation%20Montage)
    (:Is%20Valid
      (if (== (Animation|GetPlayLength (Variables|Default|GetAnimation%20Montage)) (Animation|GetPlayLength (Animation|Montage|GetCurrentActiveMontage (Components|SkeletalMesh|GetAnimInstance (Variables|/Script/Engine.Character|GetMesh (CallFunction|GetFortPawnOwner))))))
        (Animation|Sync|PlaySyncedMontage (Components|SkeletalMesh|GetAnimInstance (Variables|/Script/Engine.Character|GetMesh (CallFunction|GetFortPawnOwner))) (Animation|Montage|GetCurrentActiveMontage (Components|SkeletalMesh|GetAnimInstance (Variables|/Script/Engine.Character|GetMesh (CallFunction|GetFortPawnOwner)))) (Components|SkeletalMesh|GetAnimInstance (Variables|Default|GetSkeletalMeshComponent)) (Variables|Default|GetAnimation%20Montage))
        (Variables|Default|SetbHasEmoteStarted true)
        (else
          (Development|PrintString "Prop Montage Length Is Different Than Player Montage" true true "(R=1.000000,G=0.000000,B=0.000000,A=1.000000)" 1e+01))))))

(fn DetermineAnimationMontage
  (bind _prop_montages (Variables|Default|GetProp%20Montages))
  (bind (_mediummale_5_3bbc21ae4123606f9f42baa8fc858e6e _mediumfemaleoverride_11_fdb37cf1414b2428ddb8649e23959990 _largemaleoverride_9_5bd65eab4b1fff21192bafb35bc3b42e _largefemaleoverride_10_b237a41c4a2ec615404577a07077ebbf) (Utilities|Struct|BreakFortEmoteMontageStruct _prop_montages))
  (bind _prop_montages_2 (Variables|Default|GetProp%20Montages))
  (bind (_mediummale_5_3bbc21ae4123606f9f42baa8fc858e6e_2 _mediumfemaleoverride_11_fdb37cf1414b2428ddb8649e23959990_2) (Utilities|Struct|BreakFortEmoteMontageStruct _prop_montages_2))
  (bind _prop_montages_3 (Variables|Default|GetProp%20Montages))
  (bind _mediummale_5_3bbc21ae4123606f9f42baa8fc858e6e_3 (Utilities|Struct|BreakFortEmoteMontageStruct _prop_montages_3))
  (bind _bisfemale (LocalVariables|GetbIsFemale))
  (bind _returnvalue (select _bisfemale _largefemaleoverride_10_b237a41c4a2ec615404577a07077ebbf _largemaleoverride_9_5bd65eab4b1fff21192bafb35bc3b42e))
  (bind _asfort_player_pawn (Utilities|Casting|CastToFortPlayerPawn (CallFunction|GetFortPawnOwner)))
  (LocalVariables|SetbIsFemale (.IsFemale (FortPlayerPawn|GetGender _asfort_player_pawn)))
  (if (Utilities|Enum|Equal(Enum) (Customization|GetCharacterBodyType _asfort_player_pawn) "Large")
    (Utilities|IsValid _returnvalue
      (:Is%20Valid
        (return _returnvalue))
      (:Is%20Not%20Valid
        _)))
  (if (LocalVariables|GetbIsFemale)
    (Utilities|IsValid _mediumfemaleoverride_11_fdb37cf1414b2428ddb8649e23959990_2
      (:Is%20Valid
        (return _mediumfemaleoverride_11_fdb37cf1414b2428ddb8649e23959990_2))
      (:Is%20Not%20Valid
        _)))
  (Utilities|IsValid _mediummale_5_3bbc21ae4123606f9f42baa8fc858e6e_3
    (:Is%20Valid
      _)
    (:Is%20Not%20Valid
      (Development|PrintString "[Peristent Prop] No Valid Prop Montage for Medium Male" true true "(R=1.000000,G=0.046544,B=0.000000,A=1.000000)" 5.0)))
  (return _mediummale_5_3bbc21ae4123606f9f42baa8fc858e6e_3))

(fn ConstructionScript
  (|CallParentFunction))
